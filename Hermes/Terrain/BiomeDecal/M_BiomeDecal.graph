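# Biome decal material for Hermes terrain, built on material attributes and the Goldlands landscape layer function.
# An Apollo landscape runtime virtual texture sample feeds its BaseColor output into base color, specular, roughness and normal of a SetMaterialAttributes node.
# Those attributes enter MF_Hermes_Goldlands as its Material input.
# A biome decal render-target sample is appended to itself, but the result connects to nothing shown, and the function's mask input is unwired.
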Material /Hermes/Terrain/BiomeDecal/M_BiomeDecal
graph {
 properties {
  bAutomaticallySetUsageInEditor=false
  bDisableDepthTest=true
  bTangentSpaceNormal=false
  bUseMaterialAttributes=true
  bUseTranslucencyVertexFog=false
  usedShadingModels=Default Lit
 }
 nodes [
  n1 = TextureSample(UVs=TextureCoordinate()) {samplerType=SAMPLERTYPE_LinearColor, texture=&/RT_BiomeDecal_Tex}
  AppendVector(A=@n1.RGB, B=@n1.RGB)
  n2 = RuntimeVirtualTextureSampleParameter!(UVs, Translated World Position) {
   group=RVT
   materialType=BaseColor_Normal_Specular_YCoCg
   parameterName=RVT Sample Target
   virtualTexture=/Game/Athena/Apollo/Maps/Landscape/VT/RVT_Apollo_Landscape
   worldPositionOriginType=CameraRelative
  }
  n3 = SetMaterialAttributes!(MaterialAttributes, Base Color=@n2.BaseColor, Specular=@n2.BaseColor, Roughness=@n2.BaseColor, Normal=@n2.BaseColor, Metallic) {attributeSetTypes [
    #1
    #2
    #3
    #4
    #5
   ]}
  MaterialFunctionCall!(Material=@n3, MobileMaterial, Texture Mask, Texture UVs, IsStaticMesh) {materialFunction=/BRCoreMaterials/Hermes/Terrain/Landscape/Materials/Layers/MF_Hermes_Goldlands}
 ]
}
